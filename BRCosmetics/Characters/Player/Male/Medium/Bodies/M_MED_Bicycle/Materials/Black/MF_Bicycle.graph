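# Material function for the Bicycle character cosmetic that layers an animated suit/body transition onto incoming material attributes.
# A curve parameter moves soft bands along pre-skinned, bounds-normalized height, building an opacity mask that static switches invert or adjust per part.
# Time-seeded hash grids and an edge texture break up the suit's mask, while non-opaque suit variants add pulsing emissive and sparks.
# An optional world position offset displaces vertices within the band.
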MaterialFunction /BRCosmetics/Characters/Player/Male/Medium/Bodies/M_MED_Bicycle/Materials/Black/MF_Bicycle
graph {
 properties {description=2}
 nodes [
  n1 = MaterialFunctionCall(Float2=TextureCoordinate()) {materialFunction=&BreakOutFloat2Components}
  n2 = Multiply(A=Constant() {r=60}, B=@n1.R)
  n3 = Fmod(A=@n2, B=Constant() {r=2})
  n4 = Round(None=OneMinus(None=Saturate(None=@n3)))
  n5 = Add(A=@n4, B=Ceil(None=@n2))
  n6 = Multiply(A=@n1.G, B=Constant() {r=30})
  n7 = Ceil(None=LinearInterpolate(A=@n6, B=Add(A=@n6) {constB=0.5}, Alpha=@n4))
  n8 = ComponentMask(None=MaterialFunctionCall(Seed=Time(), Vec2=AppendVector(A=@n5, B=@n7)) {materialFunction=&MF_Hash}.Vec2) {g=true}
  n9 = FunctionInput(Preview=VertexInterpolator(VS=PreSkinnedPosition()).PS) {bUsePreviewValueAsDefault=true, inputName=Pre Skinned Local Position}
  n10 = Divide(A=@n9, B=PreSkinnedLocalBounds().Max) {constB=500}
  n11 = ComponentMask(None=@n10) {b=true}
  n12 = ScalarParameter() {group=Bicycle, parameterName=HT_L1_TransitionCurve, sliderMax=1}
  n13 = StaticSwitchParameter(True=@n12, False=Add(A=@n12, B=ScalarParameter() {group=Bicycle, parameterName=BIKE_TransitionAmount_Offset, sliderMax=1})) {group=Bicycle, parameterName=IsBody}
  n14 = ScalarParameter() {defaultValue=0.1, group=Bicycle, parameterName=BIKE_TransitionRadius}
  n15 = Multiply(A=@n14) {constB=2}
  n16 = Subtract(A=Multiply(A=StaticSwitchParameter(True=@n13, False=@n13) {group=Bicycle, parameterName=IsSuit}, B=Add(A=@n15) {constA=1}), B=@n14)
  n17 = Multiply(A=@n14) {constB=0.4}
  n18 = Constant()
  n19 = SphereMask(A=@n11, B=Subtract(A=@n16, B=@n17), Radius=@n14, Hardness=@n18)
  n20 = SphereMask(A=@n11, B=Add(A=@n16, B=@n17), Radius=@n14, Hardness=@n18)
  n21 = Add(A=@n19, B=@n20)
  n22 = ScalarParameter() {defaultValue=5, group=Bicycle, parameterName=BIKE_TransitionPower}
  n23 = Power(Base=SphereMask(A=@n11, B=@n16, Radius=@n14, Hardness=@n18), Exp=@n22) {constExponent=5}
  n24 = Add(A=OneMinus(None=Step(Y=@n11, X=@n16) {constX=0.5}), B=@n23)
  n25 = Saturate(None=@n24)
  n26 = StaticSwitchParameter(True=OneMinus(None=@n25), False=@n25) {defaultValue=true, group=Bicycle, parameterName=IsSuit}
  n27 = Multiply(A=LinearInterpolate(B=@n8, Alpha=@n21) {constA=1}, B=@n26)
  Add(A=@n27) {constB=0.4}
  Add()
  n28 = SphereMask(A=Abs(None=ComponentMask(None=@n10) {r=true}), B=Constant(), Radius=Constant() {r=0.3}, Hardness=Constant() {r=1})
  n29 = ComponentMask(None=@n10) {b=true}
  n30 = Constant() {r=0.674}
  n31 = Constant() {r=0.08}
  n32 = Constant()
  n33 = Multiply(
   A=Multiply(A=@n28, B=SphereMask(A=@n29, B=@n30, Radius=@n31, Hardness=@n32))
   B=ScalarParameter() {defaultValue=1, group=Bicycle, parameterName=BIKE_WaistThickenAmount}
  ) {constB=0}
  n34 = ComponentMask(None=@n9) {g=true}
  n35 = ScalarParameter() {defaultValue=1, group=Bicycle, parameterName=BIKE_WaistThickenMult_Back}
  n36 = AppendVector(A=ComponentMask(None=@n9) {r=true}, B=If(A=@n34, B=Constant(), "A > B"=@n34, "A == B"=@n34, "A < B"=Multiply(A=@n34, B=@n35) {constB=1.5}))
  AppendVector(
   A=StaticSwitchParameter(True=Multiply(A=@n33, B=@n36), False=Constant2Vector()) {group=Bicycle, parameterName=BIKE_WaistOffset}
   B=Constant()
  )
  n37 = ScalarParameter() {defaultValue=1, group=Bicycle, parameterName=BIKE_GridNoiseDensity}
  n38 = MaterialFunctionCall(Seed=Multiply(A=Time()) {constB=0.5}, Vec3=Ceil(None=Multiply(A=@n9, B=@n37))) {materialFunction=&MF_Hash}
  ComponentMask(None=@n38.Vec3) {b=true}
  Divide()
  n39 = GetMaterialAttributes(None=FunctionInput() {inputName=Material, inputType=FunctionInput_MaterialAttributes}) {attributeGetTypes [
    #1
    #2
    #3
    #4
   ]}
  n40 = LinearInterpolate(B=LinearInterpolate(B=@n8, Alpha=@n19) {constA=1}) {constA=1, constAlpha=0.4}
  n41 = OneMinus(None=@n26)
  n42 = ComponentMask(None=@n38.Vec3) {r=true}
  n43 = ConstantBiasScale(None=@n42) {bias=-0.2}
  n44 = Power(Base=@n19, Exp=@n22) {constExponent=5}
  n45 = StaticSwitchParameter(True=@n44, False=@n23) {group=Bicycle, parameterName=IsSuit}
  n46 = Multiply(A=Multiply(A=LinearInterpolate(A=@n43, Alpha=@n45), B=@n45), B=Constant() {r=1})
  n47 = Add(A=Multiply(A=@n19, B=Add(A=@n41, B=@n46)), B=Multiply(A=Power(Base=@n21) {constExponent=4}) {constB=0.5})
  n48 = Multiply(
   A=Multiply(A=Add(A=@n26) {constB=0.2}, B=Multiply(A=@n40, B=@n47))
   B=VectorParameter() {group=Bicycle, parameterName=BIKE_EmissiveColor, defaultValue=<rgba 1 1 1 1>}.RGB
  )
  n49 = Sine(None=Multiply(A=Time(), B=ScalarParameter() {defaultValue=1, group=Bicycle, parameterName=BIKE_PulseSpeed}))
  n50 = Multiply(
   A=ScalarParameter() {defaultValue=1, group=Bicycle, parameterName=BIKE_EmissiveMult}
   B=Add(B=Multiply(A=@n49) {constA=1, constB=0.5}) {constA=1}
  )
  n51 = VectorParameter() {group=Bicycle, parameterName=BIKE_EmissiveColor, defaultValue=<rgba 1 1 1 1>}
  n52 = Multiply(A=@n42, B=ComponentMask(None=@n38.Vec3) {g=true})
  n53 = ConstantBiasScale(None=@n52) {bias=-0.9, scale=1}
  n54 = ConstantBiasScale(None=@n52) {bias=-0.5, scale=1}
  n55 = Power(Base=Saturate(None=LinearInterpolate(A=@n53, B=@n54, Alpha=@n44))) {constExponent=0.5}
  n56 = ConstantBiasScale(None=@n55) {bias=-0.25, scale=1}
  n57 = Saturate(None=LinearInterpolate(
    A=Multiply(A=@n51.RGB, B=@n55)
    B=VectorParameter() {group=Bicycle, parameterName=BIKE_SparksSecondatyColor, defaultValue=<rgba 1 1 1 1>}.RGB
    Alpha=Ceil(None=@n56)
   ))
  n58 = ScalarParameter() {defaultValue=3, group=Bicycle, parameterName=BIKE_SparksEmissiveMult}
  n59 = Add(A=@n39.EmissiveColor, B=Add(A=Multiply(A=@n48, B=@n50), B=Multiply(A=@n57, B=@n58)))
  n60 = Subtract(A=WorldPosition().XYZ, B=ObjectPositionWS())
  n61 = Multiply(A=Multiply(A=Multiply(A=@n60, B=@n28), B=@n21), B=ScalarParameter() {defaultValue=0.1, group=Bicycle, parameterName=BIKE_WaistAmount}) {constB=0.1}
  n62 = Constant()
  n63 = TextureSampleParameter2D() {
   parameterName=BIKE_EdgeTexture
   texture=/BRCosmetics/Characters/Player/Male/Medium/Bodies/M_MED_Bicycle_Jaguar/Textures/T_Bicycle_Jaguar_Body_EDGE
  }
  n64 = Power(Base=@n63.R) {constExponent=0.3}
  n65 = ComponentMask(None=MaterialFunctionCall(Seed=Time(), Vec3=ComponentMask(None=@n10) {b=true, g=true, r=true}) {materialFunction=&MF_Hash}.Vec3) {r=true}
  n66 = Power(Base=@n21) {constExponent=1.5}
  n67 = LinearInterpolate(A=@n65, Alpha=@n66)
  n68 = Multiply(A=@n64, B=Multiply(A=LinearInterpolate(A=@n65, Alpha=@n64), B=Multiply(A=@n67, B=@n66)))
  n69 = StaticSwitchParameter(True=Saturate(None=Add(A=@n68, B=@n27)), False=Add(A=@n26, B=@n19)) {group=Bicycle, parameterName=IsSuit}
  n70 = SetMaterialAttributes!(
   MaterialAttributes=@n39.MaterialAttributes
   BaseColor=@n39.MaterialAttributes
   EmissiveColor=StaticSwitchParameter(True=@n39.EmissiveColor, False=StaticSwitchParameter(True=@n59, False=@n39.EmissiveColor) {group=Bicycle, parameterName=IsSuit}) {group=Bicycle, parameterName=IsOpaque}
   OpacityMask=@n69
   WorldPositionOffset=Add(A=StaticSwitchParameter(True=@n61, False=@n62) {group=Bicycle, parameterName=BIKE_WaistOffset}, B=@n39.WorldPositionOffset)
  ) {attributeSetTypes [
    #1
    #2
    #5
    #3
   ]}
  FunctionOutput(None=@n70)
  LinearInterpolate()
  n71 = TextureSampleParameter2D(UVs=Multiply(A=ComponentMask(None=@n10) {b=true, r=true}, B=ScalarParameter() {defaultValue=1, group=Bicycle, parameterName=BIKE_NoiseTiling1})) {group=Bicycle, parameterName=BIKE_NoiseTexture1, texture=/Game/Global/Textures/Default/Tileable_Patterns/T_Blowing_Smoke_Pattern2_Tile_DISP}
  n72 = Round(None=Power(Base=@n71.R) {constExponent=0.2})
  n73 = Saturate(None=@n24)
  Multiply(A=Power(Base=LinearInterpolate(A=@n72, Alpha=@n73)), B=@n24)
  OneMinus(None=@n13)
  Reroute(None=@n20) {desc=TRANSITION MASKHIGH}
  Saturate(None=@n69)
  Saturate(None=@n56)
 ]
}
