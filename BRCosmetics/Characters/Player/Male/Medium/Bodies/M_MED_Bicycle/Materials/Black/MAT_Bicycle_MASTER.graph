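# Master masked material for the Bicycle male medium player body cosmetic, supporting skin, cloth and clear coat shading.
# Character PreFX output passes through switchable Skin FX layers (panning emissive, silk, clear coat, metal LUT, cloth fuzz, thin film), masked by selectable SkinFX_Mask channels.
# Vertex-color red values select hideable elements, driving a world position offset that crunches vertices toward the object origin.
# MF_Bicycle and character PostFX finish the result.
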Material /BRCosmetics/Characters/Player/Male/Medium/Bodies/M_MED_Bicycle/Materials/Black/MAT_Bicycle_MASTER
graph {
 properties {
  bUseMaterialAttributes=true
  bUsedWithClothing=true
  bUsedWithMeshDeformer=true
  bUsedWithMorphTargets=true
  bUsedWithSkeletalMesh=true
  blendMode=BLEND_Masked
  pixelDepthOffsetMode=PDOM_Legacy
  shadingModel=MSM_FromMaterialExpression
  subsurfaceProfile=&SS_HeroSkin_02
  usedShadingModels=Default Lit | Clear Coat | Subsurface Profile | Cloth
 }
 nodes [
  n1 = MaterialFunctionCall() {materialFunction=&BaseFN_CharacterFunction_PreFX}
  n2 = TextureSampleParameter2D() {group=Skin FX, parameterName=SkinFX_Mask}
  n3 = StaticComponentMaskParameter(None=@n2.RGB) {defaultR=true, group=Skin FX, parameterName=Thin FIlm Channel}
  n4 = StaticSwitchParameter(
   True=MaterialFunctionCall!(
    Alpha=StaticComponentMaskParameter(None=@n2.RGB) {defaultR=true, group=Skin FX, parameterName=Panning Emissive Channel}
    isVehicle
    Gmap
    UVOffset
    Local Normal
    Local Position
    EmissiveColorOverride
    Material=@n1.Result
   ) {materialFunction=/Game/Weapons/WeaponSkins/MaterialFunction/PanningEmissive}.Result
   False=@n1.Result
  ) {group=Skin FX, parameterName=Use Panning Emissive}
  n5 = StaticSwitchParameter(
   True=MaterialFunctionCall!(Alpha=StaticComponentMaskParameter(None=@n2.RGB) {defaultR=true, group=Skin FX, parameterName=Silk Channel}, Material=@n4) {materialFunction=/Game/Weapons/WeaponSkins/MaterialFunction/MF_Silk}.Result
   False=@n4
  ) {group=Skin FX, parameterName=Use Silk}
  n6 = StaticSwitchParameter(
   True=MaterialFunctionCall!(Alpha=StaticComponentMaskParameter(None=@n2.RGB) {defaultR=true, group=Skin FX, parameterName=Clear Coat Channel}, In=@n5) {materialFunction=/Game/Weapons/WeaponSkins/MaterialFunction/Clearcoat}.Result
   False=@n5
  ) {group=Skin FX, parameterName=Use Clear Coat}
  n7 = StaticSwitchParameter(
   True=MaterialFunctionCall(Mask=StaticComponentMaskParameter(None=@n2.RGB) {defaultR=true, group=Skin FX, parameterName=Metal LUT Channel}, Materials=@n6) {materialFunction=&MF_MetalsLUT}.Result
   False=@n6
  ) {group=Skin FX, parameterName=UseMetalLUT}
  n8 = StaticSwitchParameter(
   True=MaterialFunctionCall!(
    Alpha=StaticComponentMaskParameter(None=@n2.RGB) {defaultR=true, group=Skin FX, parameterName=Cloth Fuzz Channel}
    Gmap=@n1.Gmap
    Material=@n7
   ) {materialFunction=/Game/Weapons/WeaponSkins/MaterialFunction/ClothFuzz}.Result
   False=@n7
  ) {group=Skin FX, parameterName=Use Cloth fuzz}
  n9 = GetMaterialAttributes(None=StaticSwitchParameter(
    True=MaterialFunctionCall!(Alpha=@n3, Material Attributes=@n8) {materialFunction=/Game/Weapons/WeaponSkins/MaterialFunction/ThinFilmFunction}.Result
    False=@n8
   ) {group=Skin FX, parameterName=Use Thin Film}) {attributeGetTypes [#1]}
  n10 = Subtract(A=ObjectPositionWS(), B=WorldPosition().XYZ)
  n11 = ScalarParameter() {parameterName=HT_CrunchVerts}
  n12 = ScalarParameter() {group=HideElements, parameterName=Hide Element 0_5}
  n13 = Constant() {r=0.05}
  n14 = VertexColor()
  n15 = ScalarParameter() {group=HideElements, parameterName=Hide Element 1_5}
  n16 = Constant() {r=0.15}
  n17 = Max(
   A=Max(
    A=MaterialFunctionCall!(Hide Element Toggle=@n12, VertexColorChannel=@n14.R, Vertex Value to hide=@n13) {materialFunction=/Game/Packages/Fortress_SharedMaterials/Base_Material_Functions/VertexColorMask}.Result
    B=MaterialFunctionCall(Hide Element Toggle=@n15, VertexColorChannel=@n14.R, Vertex Value to hide=@n16) {materialFunction=&VertexColorMask}.Result
   )
   B=MaterialFunctionCall(
    Hide Element Toggle=ScalarParameter() {group=HideElements, parameterName=Hide Element 2_5}
    VertexColorChannel=@n14.R
    Vertex Value to hide=Constant() {r=0.25}
   ) {materialFunction=&VertexColorMask}.Result
  )
  n18 = MaterialFunctionCall(
   Hide Element Toggle=ScalarParameter() {group=HideElements, parameterName=Hide Element 3_5}
   VertexColorChannel=@n14.R
   Vertex Value to hide=Constant() {r=0.35}
  ) {materialFunction=&VertexColorMask}
  n19 = ScalarParameter() {group=HideElements, parameterName=Hide Element 4_5}
  n20 = Constant() {r=0.45}
  n21 = Max(
   A=Max(
    A=Max(A=@n17, B=@n18.Result)
    B=MaterialFunctionCall(Hide Element Toggle=@n19, VertexColorChannel=@n14.R, Vertex Value to hide=@n20) {materialFunction=&VertexColorMask}.Result
   )
   B=MaterialFunctionCall(
    Hide Element Toggle=ScalarParameter() {group=HideElements, parameterName=Hide Element 5_5}
    VertexColorChannel=@n14.R
    Vertex Value to hide=Constant() {r=0.55}
   ) {materialFunction=&VertexColorMask}.Result
  )
  n22 = MaterialFunctionCall(
   Hide Element Toggle=ScalarParameter() {group=HideElements, parameterName=Hide Element 6_5}
   VertexColorChannel=@n14.R
   Vertex Value to hide=Constant() {r=0.65}
  ) {materialFunction=&VertexColorMask}
  n23 = ScalarParameter() {group=HideElements, parameterName=Hide Element 7_5}
  n24 = Constant() {r=0.75}
  n25 = Max(
   A=Max(
    A=Max(A=@n21, B=@n22.Result)
    B=MaterialFunctionCall(Hide Element Toggle=@n23, VertexColorChannel=@n14.R, Vertex Value to hide=@n24) {materialFunction=&VertexColorMask}.Result
   )
   B=MaterialFunctionCall(
    Hide Element Toggle=ScalarParameter() {group=HideElements, parameterName=Hide Element 8_5}
    VertexColorChannel=@n14.R
    Vertex Value to hide=Constant() {r=0.85}
   ) {materialFunction=&VertexColorMask}.Result
  )
  n26 = MaterialFunctionCall(
   Hide Element Toggle=ScalarParameter() {group=HideElements, parameterName=Hide Element 9_5}
   VertexColorChannel=@n14.R
   Vertex Value to hide=Constant() {r=0.95}
  ) {materialFunction=&VertexColorMask}
  n27 = ScalarParameter() {group=HideElements, parameterName=Hide Element 01}
  n28 = Constant() {r=0.1}
  n29 = ScalarParameter() {group=HideElements, parameterName=Hide Element 02}
  n30 = Constant() {r=0.2}
  n31 = Max(
   A=Max(
    A=MaterialFunctionCall(Hide Element Toggle=@n27, VertexColorChannel=@n14.R, Vertex Value to hide=@n28) {materialFunction=&VertexColorMask}.Result
    B=MaterialFunctionCall(Hide Element Toggle=@n29, VertexColorChannel=@n14.R, Vertex Value to hide=@n30) {materialFunction=&VertexColorMask}.Result
   )
   B=MaterialFunctionCall(
    Hide Element Toggle=ScalarParameter() {group=HideElements, parameterName=Hide Element 03}
    VertexColorChannel=@n14.R
    Vertex Value to hide=Constant() {r=0.3}
   ) {materialFunction=&VertexColorMask}.Result
  )
  n32 = MaterialFunctionCall(
   Hide Element Toggle=ScalarParameter() {group=HideElements, parameterName=Hide Element 04}
   VertexColorChannel=@n14.R
   Vertex Value to hide=Constant() {r=0.4}
  ) {materialFunction=&VertexColorMask}
  n33 = ScalarParameter() {group=HideElements, parameterName=Hide Element 05}
  n34 = Constant() {r=0.5}
  n35 = Max(
   A=Max(
    A=Max(A=@n31, B=@n32.Result)
    B=MaterialFunctionCall(Hide Element Toggle=@n33, VertexColorChannel=@n14.R, Vertex Value to hide=@n34) {materialFunction=&VertexColorMask}.Result
   )
   B=MaterialFunctionCall(
    Hide Element Toggle=ScalarParameter() {group=HideElements, parameterName=Hide Element 06}
    VertexColorChannel=@n14.R
    Vertex Value to hide=Constant() {r=0.6}
   ) {materialFunction=&VertexColorMask}.Result
  )
  n36 = MaterialFunctionCall(
   Hide Element Toggle=ScalarParameter() {group=HideElements, parameterName=Hide Element 07}
   VertexColorChannel=@n14.R
   Vertex Value to hide=Constant() {r=0.7}
  ) {materialFunction=&VertexColorMask}
  n37 = ScalarParameter() {group=HideElements, parameterName=Hide Element 08}
  n38 = Constant() {r=0.8}
  n39 = Max(
   A=Max(
    A=Max(A=@n35, B=@n36.Result)
    B=MaterialFunctionCall(Hide Element Toggle=@n37, VertexColorChannel=@n14.R, Vertex Value to hide=@n38) {materialFunction=&VertexColorMask}.Result
   )
   B=MaterialFunctionCall(
    Hide Element Toggle=ScalarParameter() {group=HideElements, parameterName=Hide Element 09}
    VertexColorChannel=@n14.R
    Vertex Value to hide=Constant() {r=0.9}
   ) {materialFunction=&VertexColorMask}.Result
  )
  n40 = MaterialFunctionCall(
   Hide Element Toggle=ScalarParameter() {group=HideElements, parameterName=Hide Element 10}
   VertexColorChannel=@n14.R
   Vertex Value to hide=Constant() {r=1}
  ) {materialFunction=&VertexColorMask}
  n41 = VertexColor()
  n42 = Max(A=Ceil(None=@n41), B=Max(A=Ceil(None=@n41.G), B=Ceil(None=@n41.B)))
  n43 = ScalarParameter() {group=HideElements, parameterName=HideElementsOnlyConsidersRedChannel}
  n44 = Multiply(A=Max(A=Max(A=@n39, B=@n40.Result), B=Max(A=@n25, B=@n26.Result)), B=OneMinus(None=Multiply(A=@n42, B=@n43)))
  n45 = Add(A=@n9.World Position Offset, B=LinearInterpolate(B=Multiply(A=@n10, B=@n11), Alpha=OneMinus(None=@n44)))
  MaterialFunctionCall(
   AO=@n1.AO
   Distance Blend=@n1.AO
   Fresnel=@n1.AO
   Gmap=@n1.AO
   Pre Skinned Local Position=@n1.AO
   Material=MaterialFunctionCall!(
    Pre Skinned Local Position=@n1.Pre skinned Local Position
    Material=SetMaterialAttributes!(MaterialAttributes=@n9.MaterialAttributes, World Position Offset=@n45) {attributeSetTypes [#1]}
   ) {materialFunction=&/MF_Bicycle}.Result
  ) {materialFunction=&BaseFN_CharacterFunction_PostFX}
 ]
}
